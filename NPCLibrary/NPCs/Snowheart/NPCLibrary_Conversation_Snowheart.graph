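# Conversation database for the Snowheart NPC in the NPC library, providing its greeting dialogue and routing to shared conversation branches.
# The entry tag leads to one speech node with a default line plus contextual lines.
# A first-ever-conversation line has its own priority curve row, while the others check Snowheart's per-match conversation count.
# The speech node feeds three link nodes jumping by entry tag to library quests, basic services and Ch5S1.D3.Q8.
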ConversationDatabase /NPCLibrary/NPCs/Snowheart/NPCLibrary_Conversation_Snowheart
entry_points [{tag=Conversation.Entry.Snowheart, destinations [#1]}]
nodes [
 ConversationEntryPointNode_7 {
  id=#1
  class=ConversationEntryPointNode
  outputs [#2]
  properties {entryTag {tagName=Conversation.Entry.Snowheart}}
 }
 Speech_C_4 {
  id=#2
  class=Speech_C
  outputs [#3, #4, #5]
  properties {
   nodeName=""
   generalConfig {
    defaultMessage="When the storm comes, find safety."
    contextualMessages [
     {
      message=Nothing shall deter me.
      contextRequirements [{requirement {class=FortControllerRequirement_IsFirstEverConversationWithNPC}}]
      priority {curve {rowName=Tandem.Speech.Priority.FirstEverInteraction}}
     }
     {
      message="When the storm comes, find safety."
      contextRequirements [{requirement {class=FortControllerRequirement_CheckNPCMatchConversationCount, properties {nPCUniqueIdTag {tagName=AISpawnerData.Type.Tandem.Snowheart}}}}]
      priority {curve {rowName=None}}
     }
     {
      message="Ice can burn, if you are not careful."
      contextRequirements [{requirement {class=FortControllerRequirement_CheckNPCMatchConversationCount, properties {nPCUniqueIdTag {tagName=AISpawnerData.Type.Tandem.Snowheart}}}}]
      priority {curve {rowName=None}}
     }
     {
      message=The road I travel is my own.
      contextRequirements [{requirement {class=FortControllerRequirement_CheckNPCMatchConversationCount, properties {nPCUniqueIdTag {tagName=AISpawnerData.Type.Tandem.Snowheart}}}}]
      priority {curve {rowName=None}}
     }
    ]
   }
  }
 }
 ConversationLinkNode_2 {id=#5, class=ConversationLinkNode, properties {remoteEntryTag {tagName=Conversation.Entry.Ch5S1.D3.Q8}}}
 ConversationLinkNode_0 {id=#4, class=ConversationLinkNode, properties {remoteEntryTag {tagName=Conversation.Entry.BasicServices}}}
 ConversationLinkNode_1 {id=#3, class=ConversationLinkNode, properties {remoteEntryTag {tagName=Conversation.Entry.NPCLibrary.Quests}}}
]
